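# Additive unlit material for Niagara mesh particles in the MusketSlinger emote FX, drawing glowing tech patterns.
# Circuit-board lines are broken up by subtracting a scrolling fluid-flame texture, then added to squares from static and panning digital-noise samples.
# A vertical gradient, offset by a strength parameter, scales the result.
# Particle color tints it through MF_FX_EmissiveAdjustments, and particle alpha multiplies the same mask.
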Material /BRCosmetics/Animation/Game/MainPlayer/Emotes/MusketSlinger/FX/Materials/M_MusketSlinger_Tech
graph {
 properties {
  bUsedWithNiagaraMeshParticles=true
  blendMode=BLEND_Additive
  pixelDepthOffsetMode=PDOM_Legacy
  shadingModel=MSM_Unlit
  usedShadingModels=Unlit
 }
 nodes [
  n1 = Multiply(
   A=TextureCoordinate()
   B=AppendVector(A=ScalarParameter() {defaultValue=2, parameterName=LinesTileX}, B=ScalarParameter() {defaultValue=2, parameterName=LinesTileY})
  )
  n2 = TextureCoordinate() {uTiling=1.4}
  n3 = AppendVector(A=ScalarParameter() {defaultValue=1.4, parameterName=BreakupTileX}, B=ScalarParameter() {defaultValue=1, parameterName=BreakupTileY})
  n4 = TextureSample(UVs=Panner(Coordinate=Multiply(A=@n2, B=@n3)) {speedY=-0.3}) {samplerType=SAMPLERTYPE_LinearGrayscale, texture=/BRCosmetics/Weapons/FORT_Melee/Pickaxe_Cube_Ninja_Male/FX/T_FluidFlameTile}
  n5 = ScalarParameter() {defaultValue=1}
  n6 = Saturate(None=Subtract(A=TextureSample(UVs=@n1) {texture=/Game/Effects/Fort_Effects/Textures/T_FX_Circuitboard}.G, B=Multiply(A=@n4.R, B=@n5)))
  n7 = Multiply(
   A=TextureCoordinate()
   B=AppendVector(A=ScalarParameter() {defaultValue=4, parameterName=SquareTileX}, B=ScalarParameter() {defaultValue=4, parameterName=SquareTileY})
  )
  n8 = TextureSample(UVs=@n7) {samplerType=SAMPLERTYPE_LinearColor, texture=/Game/Global/Textures/Default/Noise/T_Atlas_DigitalNoise_Small}
  n9 = TextureSample(UVs=Panner(Coordinate=@n7) {speedY=-0.5}) {samplerType=SAMPLERTYPE_LinearColor, texture=&T_Atlas_DigitalNoise_Small}
  n10 = Multiply(A=Add(A=Multiply(A=@n8.R, B=@n9.G)) {constB=0.5}, B=ScalarParameter() {defaultValue=0.15, parameterName=SquareStrenght}) {constB=0.15}
  n11 = ComponentMask(None=TextureCoordinate()) {g=true}
  n12 = Multiply(A=SmoothStep(Value=OneMinus(None=@n11)) {constMax=0.75}, B=SmoothStep(Value=@n11) {constMax=0.5})
  n13 = ScalarParameter() {defaultValue=1, parameterName=GradientmaskStrenght}
  n14 = Saturate(None=Multiply(A=Add(A=@n6, B=@n10), B=Add(A=@n12, B=@n13)))
  n15 = ParticleColor()
  MaterialFunctionCall(Emissive=Multiply(A=@n14, B=@n15.RGB)) {materialFunction=&MF_FX_EmissiveAdjustments}
  Multiply(A=@n14, B=@n15.A)
 ]
}
