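# Translucent unlit material for the MusketSlinger emote FX, named as an invisible material, likely for hiding mesh parts during the emote.
# It is flagged for use on skeletal meshes and mesh deformers.
# The graph holds only a single constant left at its default of zero, with no textures, parameters or other logic.
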Material /BRCosmetics/Animation/Game/MainPlayer/Emotes/MusketSlinger/FX/Materials/M_Musketslinger_Invisible
graph {
 properties {
  bUsedWithMeshDeformer=true
  bUsedWithSkeletalMesh=true
  blendMode=BLEND_Translucent
  pixelDepthOffsetMode=PDOM_Legacy
  shadingModel=MSM_Unlit
  usedShadingModels=Unlit
 }
 nodes [Constant()]
}
