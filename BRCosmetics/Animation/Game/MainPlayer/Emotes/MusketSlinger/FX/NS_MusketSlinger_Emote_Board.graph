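# Niagara effect for the board in the MusketSlinger emote, popping in warm-tinted meshes with a sprite flare and square particles.
# Box meshes overshoot in scale and flash bright by emitter age, while Flare001 gives a brief flickering glow.
# Two GPU emitters add drifting squares and a hollowed rim with an early outward push, a random tag varying spin and size.
# User.Offset positions all but the flare.
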NiagaraSystem /BRCosmetics/Animation/Game/MainPlayer/Emotes/MusketSlinger/FX/NS_MusketSlinger_Emote_Board
dependencies {
 usedDataInterfaces [&NiagaraDataInterfaceCurve, &NiagaraDataInterfaceVector2DCurve]
 usedDynamicInputs [&Vector2DFromCurve]
 usedModules [&SystemState]
}
emitters {
 Box {
  emitterUpdateScript [
   &EmitterState(Life Cycle Mode=::NewEnumerator1, Loop Behavior=::NewEnumerator1, Loop Duration Mode=::NewEnumerator1, UseLoopDelay=false)
   &SpawnBurst_Instantaneous(Spawn Count=4, Use Spawn Probability=false, UseLoopCountLimit=false)
  ]
  particleSpawnScript [&InitializeParticle(
    Color=<rgba 1 0.5231 0.1521 1>
    Lifetime=5
    Mesh Scale Mode=::NewEnumerator3
    Mesh Uniform Scale=1.05
    Position Offset=$User.Offset
    UsePositionOffset=true
    Write Material Random=true
   )]
  particleUpdateScript [
   &ParticleState(Kill Particles When Lifetime Has Elapsed=false, Let Infinitely Lived Particles Die When Emitter Deactivates=true)
   &ScaleMeshSize(Scale Factor=&Multiply_Vector(
     A=<1 1 1>
     B=&VectorFromFloat(Value=&FloatFromCurve(
       FloatCurve=DataInterface() {propertyValues {Curve {keys [<at 0 0 RCIM_Cubic>, {arriveTangent=2.8571, interpMode=RCIM_Cubic, leaveTangent=2.8571, time=0.2, value=1.15}, <at 0.35 1 RCIM_Cubic>]}}}
       CurveIndex=$Emitter.Age
      ))
    ))
   &ScaleColor(
    Scale RGB=&Multiply_Vector(
     A=<1 1 1>
     B=&VectorFromFloat(Value=&FloatFromCurve(
       FloatCurve=DataInterface() {propertyValues {Curve {keys [<at 0 0 RCIM_Cubic>, {arriveTangent=1.6667, interpMode=RCIM_Cubic, leaveTangent=1.6667, time=0.25, value=10}, <at 0.6 1 RCIM_Cubic>]}}}
       CurveIndex=$Emitter.Age
      ))
    )
    ScaleA=true
    ScaleRGB=true
   )
   &DynamicMaterialParameters(
    Index 0 Param 1=-0.5
    Module.Index0Param0Write=true
    Module.Index0Param1Write=true
    Module.Index0Param2Write=true
    Module.Index0Param3Write=true
   )
  ]
  rendererClasses [&NiagaraMeshRendererProperties]
  renderers [{rendererClass=&NiagaraMeshRendererProperties, rendererIndex=0}]
 }
 Flare001 {
  emitterUpdateScript [
   &EmitterState(Life Cycle Mode=::NewEnumerator1, Loop Behavior=::NewEnumerator1, Loop Duration=1, UseLoopDelay=false)
   &SpawnBurst_Instantaneous(Use Spawn Probability=false, UseLoopCountLimit=false)
  ]
  particleSpawnScript [
   &InitializeParticle(
    Color=<rgba 3 1.5693 0.4564 1>
    Lifetime=0.5
    Position Offset=<0 0 -9>
    Sprite Size=<220 110>
    Sprite Size Mode=::NewEnumerator4
    UsePositionOffset=true
    Write Material Random=true
   )
   &CameraOffset(Camera Offset Amount=25)
  ]
  particleUpdateScript [
   &ParticleState()
   &ScaleSpriteSize(Uniform Curve Sprite Scale=DataInterface() {propertyValues {Curve {keys [
        {arriveTangent=4.1785, interpMode=RCIM_Cubic, leaveTangent=4.1785, tangentMode=RCTM_User, time=0, value=0}
        {arriveTangent=1.1179, interpMode=RCIM_Cubic, leaveTangent=1.1179, tangentMode=RCTM_User, time=0.2042, value=0.6228}
        {arriveTangent=0.0224, interpMode=RCIM_Cubic, leaveTangent=0.0224, tangentMode=RCTM_User, time=1, value=1}
       ]}}})
   &ScaleColor(
    Scale RGB=&Multiply_VectorByFloat(
     Vector=<1 1 1>
     Float=&FloatFromCurve(
      FloatCurve=DataInterface() {propertyValues {Curve {keys [
          {arriveTangent=9.0725, interpMode=RCIM_Cubic, leaveTangent=9.0725, tangentMode=RCTM_User, time=0, value=0}
          {arriveTangent=-2.5249, interpMode=RCIM_Cubic, leaveTangent=-2.5249, tangentMode=RCTM_User, time=0.1408, value=0.9847}
          {arriveTangent=-0.0405, interpMode=RCIM_Cubic, leaveTangent=-0.0405, tangentMode=RCTM_User, time=1, value=0}
         ]}}}
      Scale Curve=&Waveform("Waveform [1]"=::NewEnumerator12, "[1] Frequency"=15.9, "[1] Amplitude Min/Max"=<0 1>, "[1] Bias"=0.8)
     )
    )
    ScaleA=true
    ScaleRGB=true
   )
  ]
  rendererClasses [&NiagaraSpriteRendererProperties]
  renderers [{rendererClass=&NiagaraSpriteRendererProperties}]
 }
 Squares {
  simTarget=GPUComputeSim
  emitterUpdateScript [
   &EmitterState()
   &SpawnRate(SpawnRate=25, Use Spawn Probability=false)
  ]
  particleSpawnScript [
   &InitializeParticle(
    Color=<rgba 3 1.5693 0.4564 1>
    Lifetime Max=1
    Lifetime Min=0.5
    Lifetime Mode=::NewEnumerator1
    Position Offset=$User.Offset
    Sprite Rotation Mode=::NewEnumerator0
    Sprite Size Mode=::NewEnumerator1
    Uniform Sprite Size Max=1
    Uniform Sprite Size Min=0.5
    UsePositionOffset=true
    Write Material Random=true
   )
   &ShapeLocation(Box Size=<45 45 3>, MaskSpawnGroup=false, Override Seed=false, Shape Primitive=::NewEnumerator2, Surface Only=false, UseOwnerScale=true)
   &AddVelocity(Velocity=<0 0 -10>, Velocity Speed Scale=2)
   /SetVariables(Particles.VisibilityTag=&RandomRangeInt(Minimum=0, Maximum=1)) {isSetParameters=true, module=SetVariables_#1}
  ]
  particleUpdateScript [
   &Drag(Drag=&RandomRangeFloat(Minimum=1, Maximum=2), Use Linear Drag=true, Use Rotational Drag=false)
   &SolveForcesAndVelocity(Clamp Velocity=false, Limit Acceleration=false)
   &ParticleState()
   &SpriteRotationRate(Rotation Rate=&MakeCustomFloatFromBool(
     Bool=&SetBoolByIntComparison(A=$Particles.VisibilityTag, B=1, Comparison Type=::NewEnumerator4)
     True Float=100
     False Float=0
    ))
   &ScaleSpriteSize(
    Uniform Curve Scale=&MakeCustomFloatFromBool(
     Bool=&SetBoolByIntComparison(A=$Particles.VisibilityTag, B=1, Comparison Type=::NewEnumerator4)
     True Float=3
     False Float=2
    )
    Uniform Curve Sprite Scale=DataInterface() {propertyValues {Curve {keys [
        {arriveTangent=5.201, interpMode=RCIM_Cubic, leaveTangent=5.201, tangentMode=RCTM_User, time=0, value=1}
        {arriveTangent=-0.2045, interpMode=RCIM_Cubic, leaveTangent=-0.2045, tangentMode=RCTM_User, time=0.2, value=1.7}
        {arriveTangent=-1.983, interpMode=RCIM_Cubic, leaveTangent=-1.983, tangentMode=RCTM_User, time=1, value=0}
       ]}}}
   )
   &DynamicMaterialParameters(
    Index 0 Param 1=0.02
    Module.Param0WriteEnabled=true
    Module.Param1WriteEnabled=true
    Module.Param2WriteEnabled=true
    Module.Param3WriteEnabled=true
   )
  ]
  rendererClasses [&NiagaraSpriteRendererProperties]
  renderers [
   {rendererClass=&NiagaraSpriteRendererProperties}
   {rendererClass=&NiagaraSpriteRendererProperties, rendererIndex=1}
  ]
 }
 SquaresRim {
  simTarget=GPUComputeSim
  emitterUpdateScript [
   &EmitterState()
   &SpawnRate(
    Spawn Probability=&FloatFromCurve(FloatCurve=DataInterface() {propertyValues {Curve {keys [<at 0.25 1>,<at 0.5 0.5>]}}}, CurveIndex=$Emitter.Age)
    SpawnRate=65
    Use Spawn Probability=true
   )
  ]
  particleSpawnScript [
   &InitializeParticle(
    Color=<rgba 3 1.5693 0.4564 1>
    Lifetime Max=1
    Lifetime Min=0.5
    Lifetime Mode=::NewEnumerator1
    Position Offset=$User.Offset
    Sprite Rotation Mode=::NewEnumerator0
    Sprite Size Mode=::NewEnumerator1
    Uniform Sprite Size Max=1
    Uniform Sprite Size Min=0.5
    UsePositionOffset=true
    Write Material Random=true
   )
   /SetVariables(Particles.VisibilityTag=&RandomRangeInt(Minimum=0, Maximum=1)) {isSetParameters=true, module=SetVariables_#1}
   &ShapeLocation(Box Size=<60 60 5>, MaskSpawnGroup=false, Override Seed=false, Shape Primitive=::NewEnumerator2, Surface Only=false, UseOwnerScale=true) {module=ShapeLocation001}
   &AddVelocity(
    Module.Use Falloff Distance=false
    Origin Offset=$User.Offset
    Origin Offset Coordinate Space=::Simulation
    Override Seed=false
    Velocity Mode=::NewEnumerator1
    Velocity Speed=&Lerp_Float(
     A=0
     B=25
     Alpha=&FloatFromCurve(FloatCurve=DataInterface() {propertyValues {Curve {keys [<at 0.25 1>,<at 0.5 0>]}}}, CurveIndex=$Emitter.Age)
    )
   )
   &KillParticlesInVolume(Box Size:Vector3f=<50 50 50>, Kill Shape=::NewEnumerator1)
  ]
  particleUpdateScript [
   &Drag(Drag=&RandomRangeFloat(Minimum=1, Maximum=2), Use Linear Drag=true, Use Rotational Drag=false)
   &SolveForcesAndVelocity(Clamp Velocity=false, Limit Acceleration=false)
   &ParticleState()
   &SpriteRotationRate(Rotation Rate=&MakeCustomFloatFromBool(
     Bool=&SetBoolByIntComparison(A=$Particles.VisibilityTag, B=1, Comparison Type=::NewEnumerator4)
     True Float=100
     False Float=0
    ))
   &ScaleSpriteSize(
    Uniform Curve Scale=&MakeCustomFloatFromBool(
     Bool=&SetBoolByIntComparison(A=$Particles.VisibilityTag, B=1, Comparison Type=::NewEnumerator4)
     True Float=3
     False Float=2
    )
    Uniform Curve Sprite Scale=DataInterface() {propertyValues {Curve {keys [
        {arriveTangent=5.201, interpMode=RCIM_Cubic, leaveTangent=5.201, tangentMode=RCTM_User, time=0, value=1}
        {arriveTangent=-0.2045, interpMode=RCIM_Cubic, leaveTangent=-0.2045, tangentMode=RCTM_User, time=0.2, value=1.7}
        {arriveTangent=-1.983, interpMode=RCIM_Cubic, leaveTangent=-1.983, tangentMode=RCTM_User, time=1, value=0}
       ]}}}
   )
   &DynamicMaterialParameters(
    Index 0 Param 1=0.02
    Module.Param0WriteEnabled=true
    Module.Param1WriteEnabled=true
    Module.Param2WriteEnabled=true
    Module.Param3WriteEnabled=true
   )
  ]
  rendererClasses [&NiagaraSpriteRendererProperties]
  renderers [
   {rendererClass=&NiagaraSpriteRendererProperties}
   {rendererClass=&NiagaraSpriteRendererProperties, rendererIndex=1}
  ]
 }
}
summary {
 systemName=NS_MusketSlinger_Emote_Board
 emitters [
  {emitterName=Box, rendererClasses [&NiagaraMeshRendererProperties]}
  {emitterName=Squares, simTarget=GPUComputeSim, rendererClasses [&NiagaraSpriteRendererProperties]}
  {emitterName=Flare001, rendererClasses [&NiagaraSpriteRendererProperties]}
  {emitterName=SquaresRim, simTarget=GPUComputeSim, rendererClasses [&NiagaraSpriteRendererProperties]}
 ]
 userVariables [User.Offset {type=Vector3f, description="", defaultValue {struct=Vector3f, value=<0 0 -5.5>}}]
}
