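# Niagara system for a short, warm-toned result flash in the Musket Slinger emote.
# Six burst emitters layer a lit flare, SubUV light rays, line and ray meshes, twinkling sparkles and GPU squares.
# Curves drive scale, color fades and mesh material parameters, while waveforms flicker the sparkles.
# Squares use two sprite renderers and a random VisibilityTag that varies each particle's spin and size.
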NiagaraSystem /BRCosmetics/Animation/Game/MainPlayer/Emotes/MusketSlinger/FX/NS_Musket_Slinger_ResultFlash
dependencies {
 usedDataInterfaces [
  &NiagaraDataInterfaceCurve
  &NiagaraDataInterfaceVector2DCurve
  &NiagaraDataInterfaceSpriteRendererInfo
 ]
 usedDynamicInputs []
 usedModules [&SystemState]
}
emitters {
 Flare001 {
  emitterUpdateScript [
   &EmitterState()
   &SpawnBurst_Instantaneous(Use Spawn Probability=false, UseLoopCountLimit=false)
  ]
  particleSpawnScript [
   &InitializeParticle(
    Color=<rgba 3 1.5693 0.4564 1>
    Lifetime=0.8
    Sprite Size Mode=::NewEnumerator3
    Uniform Sprite Size=25.2211
    UsePositionOffset=false
    Write Material Random=true
   )
   &CameraOffset(Camera Offset Amount=5)
  ]
  particleUpdateScript [
   &ParticleState()
   &ScaleSpriteSize(Uniform Curve Sprite Scale=DataInterface() {propertyValues {Curve {keys [
        {arriveTangent=5.6485, interpMode=RCIM_Cubic, leaveTangent=5.6485, tangentMode=RCTM_User, time=0, value=0}
        {arriveTangent=-0.1641, interpMode=RCIM_Cubic, leaveTangent=-0.1641, tangentMode=RCTM_User, time=0.5, value=1}
       ]}}})
   &ScaleColor(
    Scale RGB=&Multiply_VectorByFloat(
     Vector=<1 1 1>
     Float=&FloatFromCurve(FloatCurve=DataInterface() {propertyValues {Curve {keys [
          {arriveTangent=9.0725, interpMode=RCIM_Cubic, leaveTangent=9.0725, tangentMode=RCTM_User, time=0, value=0}
          {arriveTangent=-0.2385, leaveTangent=-0.2385, time=0.1408, value=0.9847}
          {arriveTangent=-0.0284, interpMode=RCIM_Cubic, leaveTangent=-0.0284, tangentMode=RCTM_User, time=0.3, value=1.025}
          {interpMode=RCIM_Cubic, leaveTangent=-0.0405, time=1, value=0}
         ]}}})
    )
    ScaleA=true
    ScaleRGB=true
   )
   /SetVariables(Particles.Flaresize2:Vector2f=&Multiply_Vector2D(A=$Particles.SpriteSize, B=<1.5149 0.25>)) {isSetParameters=true}
   &FN_Light_Attributes(Create Exponent Attribute=true, Create Radius Attribute=true, Exponent=4, Radius=50)
  ]
  rendererClasses [&NiagaraSpriteRendererProperties, &NiagaraLightRendererProperties]
  renderers [
   {rendererClass=&NiagaraSpriteRendererProperties}
   {rendererClass=&NiagaraLightRendererProperties, rendererIndex=1}
  ]
 }
 LightRays {
  emitterUpdateScript [
   &EmitterState()
   &SpawnBurst_Instantaneous(Spawn Count=10, Use Spawn Probability=false, UseLoopCountLimit=false)
  ]
  particleSpawnScript [
   &InitializeParticle(
    Color=<rgba 5 2.6154 0.7607 1>
    Lifetime Max=0.4
    Lifetime Min=0.15
    Lifetime Mode=::NewEnumerator1
    Sprite Rotation Angle=&RandomRangeFloat(Minimum=0.25, Maximum=-0.25)
    Sprite Rotation Mode=::NewEnumerator1
    Sprite Size Max=<18 25>
    Sprite Size Min=<12 15>
    Sprite Size Mode=::NewEnumerator2
    UsePositionOffset=false
    Write Material Random=true
   ) {module=InitializeParticle001}
   &DynamicMaterialParameters(
    Index 0 Param 1=50
    Module.Param0WriteEnabled=true
    Module.Param1WriteEnabled=true
    Module.Param2WriteEnabled=true
    Module.Param3WriteEnabled=true
   )
   &SubUVAnimation(
    Evaluation Type=::NewEnumerator0
    "Sprite Renderer "=DataInterface() {propertyValues {SpriteRenderer=&self:LightRays_2.NiagaraSpriteRendererProperties_0}}
    SubUV Animation Mode:ENiagara_SubUVLookupModeV2=::NewEnumerator2
    UseEndFrame=false
    UseStartFrame=false
   )
   &CameraOffset(Camera Offset Amount=5)
  ]
  particleUpdateScript [
   &ParticleState()
   &ScaleColor(
    Scale Alpha=&FloatFromCurve(FloatCurve=DataInterface() {propertyValues {Curve {keys [
         {interpMode=RCIM_Cubic, leaveTangent=-0.0594, tangentMode=RCTM_Break, time=-0.0029, value=1.0303}
         {arriveTangent=-0.936, interpMode=RCIM_Cubic, leaveTangent=-0.936, tangentMode=RCTM_User, time=0.0887, value=0.3527}
         <at 1.0011 0 RCIM_Cubic>
        ]}}})
    ScaleA=true
    ScaleRGB=true
   )
   &ScaleSpriteSize(Scale Factor=&Vector2DFromCurve(
     Vector2Curve=DataInterface() {propertyValues {
       XCurve {keys [<at 0 0 RCIM_Cubic>, {arriveTangent=0.25, interpMode=RCIM_Cubic, leaveTangent=0.25, time=0.4, value=0.75}, <at 1 0.25 RCIM_Cubic>]}
       YCurve {keys [{arriveTangent=1.6918, interpMode=RCIM_Cubic, leaveTangent=1.6918, tangentMode=RCTM_User, time=-0.0006, value=0.4523}, <at 1 1>]}
      }}
     ScaleCurve=<1 1>
    ))
  ]
  rendererClasses [&NiagaraSpriteRendererProperties]
  renderers [{rendererClass=&NiagaraSpriteRendererProperties}]
 }
 Lines {
  emitterUpdateScript [
   &EmitterState()
   &SpawnBurst_Instantaneous(Use Spawn Probability=false, UseLoopCountLimit=false)
  ]
  particleSpawnScript [
   &InitializeParticle(
    Color=<rgba 3 3 3 1>
    Lifetime=0.24
    Mesh Scale Mode=::NewEnumerator3
    Mesh Uniform Scale=0.75
    UsePositionOffset=false
    Write Material Random=true
   )
   &InitialMeshOrientation(
    Mesh Orientation Mode=::NewEnumerator7
    Rotation=&RandomRangeVector(Minimum=<0 0 0>, Maximum=<1 0 0>)
    Use Rotation Vector=true
   )
   &CameraOffset(Camera Offset Amount=-7.5)
  ]
  particleUpdateScript [
   &ParticleState()
   &DynamicMaterialParameters(
    Index 0 Param 1=&FloatFromCurve(FloatCurve=DataInterface() {propertyValues {Curve {keys [
         {interpMode=RCIM_Cubic, leaveTangent=0.363, tangentMode=RCTM_Break, time=0, value=-0.69}
         {arriveTangent=2.685, interpMode=RCIM_Cubic, tangentMode=RCTM_Break, time=1, value=0.55}
        ]}}})
    Module.Param0WriteEnabled=true
    Module.Param1WriteEnabled=true
    Module.Param2WriteEnabled=true
    Module.Param3WriteEnabled=true
   )
   &ScaleColor(
    Scale Alpha=&FloatFromCurve(FloatCurve=DataInterface() {propertyValues {Curve {keys [
         {interpMode=RCIM_Cubic, leaveTangent=-2.9428, tangentMode=RCTM_Break, time=0, value=1}
         {arriveTangent=0.0096, interpMode=RCIM_Cubic, tangentMode=RCTM_Break, time=1, value=0}
        ]}}})
    Scale RGB=&Multiply_VectorByFloat(
     Vector=<1 1 1>
     Float=&FloatFromCurve(FloatCurve=DataInterface() {propertyValues {Curve {keys [
          {arriveTangent=-7.4679, interpMode=RCIM_Cubic, leaveTangent=-7.4679, tangentMode=RCTM_User, time=0, value=4}
          {arriveTangent=-0.8552, interpMode=RCIM_Cubic, leaveTangent=-0.8552, tangentMode=RCTM_User, time=1, value=0}
         ]}}})
    )
    ScaleA=true
    ScaleRGB=true
   )
  ]
  rendererClasses [&NiagaraMeshRendererProperties]
  renderers [{rendererClass=&NiagaraMeshRendererProperties, rendererIndex=0}]
 }
 Rays {
  emitterUpdateScript [
   &EmitterState()
   &SpawnBurst_Instantaneous(Use Spawn Probability=false, UseLoopCountLimit=false)
  ]
  particleSpawnScript [
   &InitializeParticle(
    Color=<rgba 1 0.6843 0.1036 1>
    Mesh Scale=<0.01 0.01 0.9>
    Mesh Scale Mode=::NewEnumerator4
    UsePositionOffset=false
    Write Material Random=true
   )
   &InitialMeshOrientation(Mesh Orientation Mode=::NewEnumerator2, Rotation=<0.5 0 0>, Rotation Coordinate Space=::NewEnumerator1, Use Rotation Vector=true)
   &CameraOffset(Camera Offset Amount=2)
  ]
  particleUpdateScript [
   &ParticleState()
   &ScaleMeshSize(Scale Factor=&MakeVector(
     X=&FloatFromCurve(FloatCurve=DataInterface() {propertyValues {Curve {keys [<at 0 0 RCIM_Cubic>, {arriveTangent=4.5519, interpMode=RCIM_Cubic, leaveTangent=4.5519, tangentMode=RCTM_User, time=0.5, value=1}]}}})
     Y=&FloatFromCurve(FloatCurve=DataInterface() {propertyValues {Curve {keys [<at 0 0 RCIM_Cubic>, {arriveTangent=3.8312, interpMode=RCIM_Cubic, leaveTangent=3.8312, tangentMode=RCTM_User, time=0.5, value=1}]}}})
     Z=&FloatFromCurve(FloatCurve=DataInterface() {propertyValues {Curve {keys [{arriveTangent=2.0621, interpMode=RCIM_Cubic, leaveTangent=2.0621, tangentMode=RCTM_User, time=0, value=0}, <at 0.4 1 RCIM_Cubic>]}}})
    ))
   &DynamicMaterialParameters(
    Index 0 Param 1=&Add_Float(
     A=&Multiply_Float(A=$Emitter.Age, B=0.25)
     B=&RandomRangeFloat(Minimum=0, Maximum=1)
    )
    Index 0 Param 2=&FloatFromCurve(FloatCurve=DataInterface() {propertyValues {Curve {keys [{tangentMode=RCTM_User, time=0, value=0}, {tangentMode=RCTM_User, time=1, value=1}]}}})
    Index 0 Param 3=&Lerp_Float(
     A=0
     B=320
     Alpha=&FloatFromCurve(FloatCurve=DataInterface() {propertyValues {Curve {keys [
          {arriveTangent=1.4483, interpMode=RCIM_Cubic, leaveTangent=1.4483, tangentMode=RCTM_User, time=0, value=0}
          {arriveTangent=1.2944, interpMode=RCIM_Cubic, leaveTangent=1.2944, tangentMode=RCTM_User, time=0.3, value=0.625}
          <at 1 1>
         ]}}})
    )
    Module.Index0Param0Write=true
    Module.Index0Param1Write=true
    Module.Index0Param2Write=true
    Module.Index0Param3Write=true
   )
   &ScaleColor(
    Scale RGB=&Multiply_VectorByFloat(
     Vector=<1 1 1>
     Float=&FloatFromCurve(FloatCurve=DataInterface() {propertyValues {Curve {keys [
          {arriveTangent=9.0725, interpMode=RCIM_Cubic, leaveTangent=9.0725, tangentMode=RCTM_User, time=0, value=0}
          {arriveTangent=-2.5249, interpMode=RCIM_Cubic, leaveTangent=-2.5249, tangentMode=RCTM_User, time=0.1408, value=0.9847}
          {arriveTangent=-0.0405, interpMode=RCIM_Cubic, leaveTangent=-0.0405, tangentMode=RCTM_User, time=1, value=0}
         ]}}})
    )
    ScaleA=true
    ScaleRGB=true
   )
  ]
  rendererClasses [&NiagaraMeshRendererProperties]
  renderers [{rendererClass=&NiagaraMeshRendererProperties, rendererIndex=0}]
 }
 Sparkles {
  emitterUpdateScript [
   &EmitterState()
   &SpawnBurst_Instantaneous(Spawn Count=5, Use Spawn Probability=false, UseLoopCountLimit=false)
  ]
  particleSpawnScript [
   &InitializeParticle(
    Color=<rgba 6 3.144 0.7188 1>
    Lifetime Max=1
    Lifetime Min=0.3
    Lifetime Mode=::NewEnumerator1
    Sprite Size Mode=::NewEnumerator3
    Uniform Sprite Size=12
    UsePositionOffset=false
    Write Material Random=true
   )
   &ShapeLocation(
    Cone Inner Angle=25
    Cone Length=25
    MaskSpawnGroup=false
    Override Seed=false
    Rotation Coordinate Space=::World
    Shape Primitive=::NewEnumerator5
    UseOwnerScale=true
   )
   &AddVelocityFromPoint(Module.Use Falloff Distance=false, Velocity Strength=&UniformRangedFloat(Minimum=0, Maximum=200))
   &CameraOffset(Camera Offset Amount=10)
  ]
  particleUpdateScript [
   &ParticleState()
   &ScaleSpriteSize(
    Uniform Curve Scale=&Waveform("Waveform [1]"=::NewEnumerator9, "[1] Frequency"=12, "[1] Amplitude Min/Max"=<0 1>)
    Uniform Curve Sprite Scale=DataInterface() {propertyValues {Curve {keys [
        {arriveTangent=9.0725, interpMode=RCIM_Cubic, leaveTangent=9.0725, tangentMode=RCTM_User, time=0, value=0}
        {arriveTangent=-0.1942, leaveTangent=-0.1942, time=0.15, value=1}
        {arriveTangent=-1.1765, interpMode=RCIM_Cubic, leaveTangent=-1.1765, time=0.7, value=1}
        {interpMode=RCIM_Cubic, leaveTangent=-0.0405, time=1, value=0}
       ]}}}
   )
   &Drag(Drag=&RandomRangeFloat(Minimum=3, Maximum=6), Use Linear Drag=true, Use Rotational Drag=true)
   &InheritSourceMovement(
    Applied Position Scale=&Multiply_VectorByFloat(
     Vector=<1 1 1>
     Float=&FloatFromCurve(FloatCurve=DataInterface() {propertyValues {Curve {keys [
          {arriveTangent=-1.2628, interpMode=RCIM_Cubic, leaveTangent=-1.2628, tangentMode=RCTM_User, time=0, value=1}
          {interpMode=RCIM_Cubic, tangentMode=RCTM_User, time=1, value=0}
         ]}}})
    )
    ApplyForce=false
    ApplyPosition=true
    ApplyPositionFalloff=false
   )
   &SolveForcesAndVelocity(Clamp Velocity=false, Limit Acceleration=false)
   &ScaleColor(
    Scale Alpha=&FloatFromCurve(FloatCurve=DataInterface() {propertyValues {Curve {keys [<at 0 1>,<at 1 0>]}}})
    Scale RGB=&Multiply_VectorByFloat(
     Vector=<1 1 1>
     Float=&Waveform(
      "Waveform [1]"=::NewEnumerator9
      "[1] Phase Scale"=2
      "[1] Frequency"=&RandomRangeFloat(Minimum=10, Maximum=20)
      "[1] Amplitude Min/Max"=<0 1>
     )
    )
    ScaleA=true
    ScaleRGB=true
   )
  ]
  rendererClasses [&NiagaraSpriteRendererProperties]
  renderers [{rendererClass=&NiagaraSpriteRendererProperties}]
 }
 Squares {
  simTarget=GPUComputeSim
  emitterUpdateScript [
   &EmitterState()
   &SpawnBurst_Instantaneous(Spawn Count=10, Use Spawn Probability=false, UseLoopCountLimit=false)
  ]
  particleSpawnScript [
   &InitializeParticle(
    Color=<rgba 3 1.5693 0.4564 1>
    Lifetime Max=0.75
    Lifetime Min=0.35
    Lifetime Mode=::NewEnumerator1
    Sprite Rotation Mode=::NewEnumerator0
    Sprite Size Mode=::NewEnumerator1
    Uniform Sprite Size Max=1
    Uniform Sprite Size Min=0.5
    UsePositionOffset=true
    Write Material Random=true
   )
   /SetVariables(Particles.VisibilityTag=&RandomRangeInt(Minimum=0, Maximum=1)) {isSetParameters=true}
   &ShapeLocation(
    Cone Inner Angle=25
    MaskSpawnGroup=false
    Override Seed=false
    Rotation Coordinate Space=::World
    Shape Primitive=::NewEnumerator5
    UseOwnerScale=true
   )
   &AddVelocity(
    Cone Axis=<0 0 1>
    Module.UseConeFalloff=false
    Module.UseVelDistribution=false
    Override Seed=false
    Rotation Coordinate Space=::World
    Velocity Mode=::NewEnumerator2
    Velocity Speed=&RandomRangeFloat(Minimum=25, Maximum=120)
   )
  ]
  particleUpdateScript [
   &ParticleState()
   &Drag(Drag=&RandomRangeFloat(Minimum=3, Maximum=6), Use Linear Drag=true, Use Rotational Drag=false)
   &SolveForcesAndVelocity(Clamp Velocity=false, Limit Acceleration=false)
   &SpriteRotationRate(Rotation Rate=&MakeCustomFloatFromBool(
     Bool=&SetBoolByIntComparison(A=$Particles.VisibilityTag, B=1, Comparison Type=::NewEnumerator4)
     True Float=100
     False Float=0
    ))
   &ScaleSpriteSize(
    Uniform Curve Scale=&MakeCustomFloatFromBool(
     Bool=&SetBoolByIntComparison(A=$Particles.VisibilityTag, B=1, Comparison Type=::NewEnumerator4)
     True Float=3
     False Float=2
    )
    Uniform Curve Sprite Scale=DataInterface() {propertyValues {Curve {keys [
        {arriveTangent=5.201, interpMode=RCIM_Cubic, leaveTangent=5.201, tangentMode=RCTM_User, time=0, value=1}
        {arriveTangent=-0.2045, interpMode=RCIM_Cubic, leaveTangent=-0.2045, tangentMode=RCTM_User, time=0.2, value=1.7}
        {arriveTangent=-1.983, interpMode=RCIM_Cubic, leaveTangent=-1.983, tangentMode=RCTM_User, time=1, value=0}
       ]}}}
   )
   &DynamicMaterialParameters(
    Index 0 Param 1=0.02
    Module.Param0WriteEnabled=true
    Module.Param1WriteEnabled=true
    Module.Param2WriteEnabled=true
    Module.Param3WriteEnabled=true
   )
  ]
  rendererClasses [&NiagaraSpriteRendererProperties]
  renderers [
   {rendererClass=&NiagaraSpriteRendererProperties}
   {rendererClass=&NiagaraSpriteRendererProperties, rendererIndex=1}
  ]
 }
}
summary {
 systemName=NS_Musket_Slinger_ResultFlash
 emitters [
  {emitterName=Squares, simTarget=GPUComputeSim, rendererClasses [&NiagaraSpriteRendererProperties]}
  {emitterName=Flare001, rendererClasses [&NiagaraSpriteRendererProperties, &NiagaraLightRendererProperties]}
  {emitterName=LightRays, rendererClasses [&NiagaraSpriteRendererProperties]}
  {emitterName=Lines, rendererClasses [&NiagaraMeshRendererProperties]}
  {emitterName=Sparkles, rendererClasses [&NiagaraSpriteRendererProperties]}
  {emitterName=Rays, rendererClasses [&NiagaraMeshRendererProperties]}
 ]
 userVariables [
  User.LoopsDelay {type=NiagaraFloat, defaultValue=0, description=""}
  User.Offset {type=Vector3f, description="", defaultValue {struct=Vector3f, value=<0 0 -5.5>}}
 ]
}
